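; Blueprint interface for DemonMaskFire weapon actors, exposing homing and projectile hooks.
; It declares a getter and setter for the weapon's homing target actor, plus a getter for maximum projectile speed.
; A further setter passes in a B_Prj_DemonMaskFire projectile reference.
; The interface holds only signatures, so implementing weapons supply the behavior.
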(fn GetWeaponHomingTarget
  (return 0))

(fn GetWeaponMaxProjectileSpeed
  (return 0.0))

(fn SetWeaponHomingTarget
  (params (InHomingTarget :type Actor%20Object%20Reference)))

(fn SetWeaponProjectile
  (params (InProjectile :type B%20Prj%20Demon%20Mask%20Fire%20Object%20Reference)))
