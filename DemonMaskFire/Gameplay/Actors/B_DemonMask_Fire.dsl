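; B_DemonMask_Fire is the fire demon mask weapon, managing reticle, homing targeting, ammo and victim warnings for its projectile.
; With snap targeting enabled, a local timer finds the aimed-at target and feeds the target-state component and homing actor.
; A replicated lock target toggles a warning camera modifier on the victim through OnRep.
; Ammo events play recharge cues and remove depleted masks after a delay, while unequip and EndPlay reset targeting and warnings.
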(event Custom|DemonFireMask_RefillAmmo
  ;[0 0.8 1 0.5] Cheats [ Cheat ke* event ]
  (FortWeapon|ModAmmo 3))

(event Custom|DemonMaskFire_AdjustProjSpawnLoc
  (params (SpawnFromHead :type Boolean))
  ;[0 0.8 1 0.5] Cheats [ Cheat ke* event ]
  (Variables|Projectile|SetSpawn%20from%20Head SpawnFromHead))

(event Custom|DemonMaskFire_RefillAmmoRepeat
  (params (CheatEnabled :type Boolean))
  ;[0 0.8 1 0.5] Cheats [ Cheat ke* event ]
  (if CheatEnabled
    (bind _returnvalue (Utilities|Time|SetTimerbyEvent (delegate Custom|DemonFireMask_RefillAmmo) 1.0 true))
    (Variables|Cheat|SetCheatRefillAmmoTimer _returnvalue)
    (else
      (Utilities|Time|ClearandInvalidateTimerbyHandle (Variables|Cheat|GetCheatRefillAmmoTimer)))))

(event Custom|DemonMaskFire_RemoveCooldown)

(event EventBeginPlay
  (|CallParentFunction)
  (Utilities|FlowControl|Sequence
    (:then_0
      ;[0.8281 0.3911 0 0.5] VictimWarning
      (Variables|TargetedVictim|SetDisableVictimWarning (ScalableFloat|GetValueasBool (Variables|Hotfix|GetHF_DisableVictimWarning)))
      (Variables|Default|SetAllowHoming (ScalableFloat|GetValueasBool (Variables|Hotfix|GetRow_HomingEnabled)))
      (Variables|Default|SetAllowSnapToWorldReticleTargeting (ScalableFloat|GetValueasBool (Variables|Hotfix|GetRow_SnapToWorldReticleTargetingEnabled))))))

(event EventEndPlay
  (params (EndPlayReason :type EEndPlayReason%20Enum :default "Destroyed"))
  (|CallParentFunction EndPlayReason)
  (if (ScalableFloat|GetValueasBool (Variables|Default|GetHF_CleaningTimersOnUnequipAndEndPlay))
    (CallFunction|ClearTimers)
    (else
      (Utilities|Time|ClearandInvalidateTimerbyHandle (Variables|Reticle|GetReticleTickTimer))))
  ;[0.8281 0.3911 0 0.5] VictimWarning
  (CallFunction|DisableVictimWarningonEndStates)
  (Variables|Targeting|SetTargetHomingActor 0)
  ;[0.8854 0.5285 0 0.5] If mask is dropped while targeting attached and then any mask picked back up, targeting will attach to 0 0 0
  ; Reset targeting attach point so we don't point to 0 0 0 in world
  (Class|FortWeaponComponentTargetState|UpdateTargetInfo :self (Variables|Default|GetFortWeaponComponent_TargetState) :bIsTargetValid false)
  (Ability|Async|EndAction (Variables|Default|GetPrimary%20Input%20Pressed%20Async%20Action)))

(event FortWeapon|EventOnUnEquippedAndDetached
  (|CallParentFunction)
  (EventDispatchers|UnbindEventfromOnConsumeAmmo (EventDispatchers|CreateEvent :!event Custom|OnConsumeAmmoTracking))
  (EventDispatchers|UnbindEventfromOnLocalAmmoChanged (EventDispatchers|CreateEvent :!event Custom|OnLocalAmmoChanged_Event))
  (if (ScalableFloat|GetValueasBool (Variables|Default|GetHF_CleaningTimersOnUnequipAndEndPlay))
    (CallFunction|ClearTimers)
    (else
      (Utilities|Time|ClearandInvalidateTimerbyHandle (Variables|Reticle|GetReticleTickTimer))))
  ;[0.8281 0.3911 0 0.5] VictimWarning
  (CallFunction|DisableVictimWarningonEndStates)
  (Variables|Targeting|SetTargetHomingActor 0)
  (GameplayCue|ExecuteGameplayCue (Variables|Default|GetAs%20Fort%20Player%20Pawn) (Variables|Default|GetUnequip_GC))
  ;[0.8854 0.5285 0 0.5] If mask is dropped while targeting attached and then any mask picked back up, targeting will attach to 0 0 0
  ; Reset targeting attach point so we don't point to 0 0 0 in world
  (Class|FortWeaponComponentTargetState|UpdateTargetInfo :self (Variables|Default|GetFortWeaponComponent_TargetState) :bIsTargetValid false)
  (Ability|Async|EndAction (Variables|Default|GetPrimary%20Input%20Pressed%20Async%20Action)))

(event FortWeapon|EventOnEquippedAndAttached
  (bind _as_fort_player_pawn (Variables|Default|GetAs%20Fort%20Player%20Pawn))
  (|CallParentFunction)
  (bind _asfort_player_pawn (Utilities|Casting|CastToFortPlayerPawn (Game|GetInstigator)))
  (Variables|Default|SetAs%20Fort%20Player%20Pawn _asfort_player_pawn)
  (Utilities|FlowControl|Sequence
    (:then_0
      (Variables|Projectile|SetPrjMaxSpeed (ScalableFloat|GetValueatLevel (Variables|Hotfix|GetRow_MaxSpeed)))
      (EventDispatchers|BindEventtoOnConsumeAmmo (EventDispatchers|CreateEvent :!event Custom|OnConsumeAmmoTracking))
      (EventDispatchers|BindEventtoOnLocalAmmoChanged (EventDispatchers|CreateEvent :!event Custom|OnLocalAmmoChanged_Event))
      (bind _output_get (Variables|Ammo|SetLocal%20Mag%20Ammo (FortWeapon|GetMagazineAmmoCount)))
      (Variables|Ammo|SetPrevious%20Local%20Mag%20Ammo _output_get)
      ;[0.9375 0.8285 0 0.5] Set total ammo count for UI reticle usage
      ; TODO: we should have a getter for total mags/clips based on weapon data
      ; Hotfixing the values for now in case we hotfix total ammo in live, we can have the UI respond in tandem
      (bind _asfort_player_controller (Utilities|Casting|CastToFortPlayerController (Pawn|GetController (Variables|Default|GetAs%20Fort%20Player%20Pawn))))
      (if (> (Inventory|GetItemQuantityOnPlayer _asfort_player_controller (Variables|Default|GetNormal%20Item%20Def)) 0)
        (Variables|Default|SetTotalAmmoCount (Math|Conversions|ToByte(Integer) (ScalableFloat|GetValueasInteger (Variables|Hotfix|GetRow_NormalWIDTotalAmmo))))
        (else
          (Variables|Default|SetTotalAmmoCount (Math|Conversions|ToByte(Integer) (ScalableFloat|GetValueasInteger (Variables|Hotfix|GetRow_MythicWIDTotalAmmo))))))
      (bind _asynctaskproxy (Ability|Async|WaitGameplayEventToActor (Variables|Default|GetAs%20Fort%20Player%20Pawn) "(TagName=\"Event.Primary.Pressed\")")
        (:then
          (Variables|Default|SetPrimary%20Input%20Pressed%20Async%20Action _asynctaskproxy))
        (:EventReceived
          (if (== (FortWeapon|GetMagazineAmmoCount) 0)
            (if (Pawn|IsLocallyViewed _as_fort_player_pawn)
              (Audio|SpawnSoundAttached "/DemonMaskFire/Assets/Audio/DemonMaskFire_Player/SC_Mask_Demo_Fire_Player_NotReady.SC_Mask_Demo_Fire_Player_NotReady" (Variables|Visual|GetWeaponMesh))
              (FailedAction|UnabletoPerformAction _as_fort_player_pawn true (Variables|Default|GetFailed%20Reason)))))))
    (:then_1
      (if (Variables|Default|GetAllowSnapToWorldReticleTargeting)
        (Utilities|Array|AddUnique (Variables|Targeting|GetActors%20to%20Ignore%20for%20Targeting) (Variables|Default|GetAs%20Fort%20Player%20Pawn))
        (if (Pawn|IsLocallyViewed (Variables|Default|GetAs%20Fort%20Player%20Pawn))
          (bind _returnvalue (Utilities|Time|SetTimerbyEvent (delegate Custom|Event%20Update%20Reticle%20Targeting) (Variables|Targeting|GetUpdateReticleTargetingRate) true true))
          (Variables|Targeting|SetReticle%20Targeting%20Timer _returnvalue))))
    (:then_2
      ;[0.9032 0.9115 0 0.5] TEMP Reticle Start
      (Networking|CanExecuteCosmeticEvents
        (:True
          (Default|UpdateReticleVisibility))))))

(event Custom|Event%20Update%20Reticle%20Targeting
  (Default|UpdateReticleTargeting))

(event FortWeapon|EventOnPlayImpactFX
  (params (HitResult :type Hit%20Result%20Structure%20%28by%20ref%29) (ImpactPhysicalSurface :type EPhysicalSurface%20Enum :default "SurfaceType_Default") (SpawnedPSC :type FXSystem%20Component%20Object%20Reference))
  (|CallParentFunction HitResult ImpactPhysicalSurface SpawnedPSC)
  (Effects|Components|ParticleSystem|SetFloatParameter SpawnedPSC "DebrisAmount" (Variables|Default|GetDebrisAmount)))

(event Custom|ReticleTick
  ;[0.8281 0.3911 0 0.5] VictimWarning
  (if (not (Variables|TargetedVictim|GetDisableVictimWarning))
    (Default|TickVictimWarning)))

(event Custom|Event%20ServerSetTargetHomingActor
  (params (New%20Target :type Actor%20Object%20Reference))
  (bind _output_get (Variables|Targeting|SetTargetHomingActor New%20Target))
  (if (not (Variables|TargetedVictim|GetDisableVictimWarning))
    (CallFunction|TellServerLockOn _output_get)
    (Development|PrintString (Utilities|GetDisplayName _output_get))))

(event Custom|Tell%20Server%20Lock%20On
  (params (Locked%20On%20Actor :type Actor%20Object%20Reference))
  ;[0.8281 0.3911 0 0.5] VictimWarning
  (Utilities|IsValid Locked%20On%20Actor
    (:Is%20Valid
      (bind _asfort_pawn (Utilities|Casting|CastToFortPawn Locked%20On%20Actor)
        (:then
          (bind _output_get (Variables|TargetedVictim|SetRepLockTarget _asfort_pawn)))
        (:CastFailed
          (Variables|TargetedVictim|SetRepLockTarget 0)))
      (Networking|ForceNetUpdate _output_get))
    (:Is%20Not%20Valid
      _)))

(event Custom|OnLocalAmmoChanged_Event
  (params (LocalCount :type Integer) (LocalRemaining :type Integer))
  (if (> LocalCount (Variables|Ammo|GetPrevious%20Local%20Mag%20Ammo))
    (GameplayCue|ExecuteGameplayCue (Variables|Default|GetAs%20Fort%20Player%20Pawn) (Variables|Ammo|GetRecharge%20Gameplay%20Cue%20Tag))))

(event Custom|OnConsumeAmmoTracking
  (params (Quantity :type Integer))
  (bind _returnvalue (FortWeapon|GetMagazineAmmoCount))
  ;[0 0.9688 0.1076 0.5] Check ammo remaining, if 0, discard the item. Persist in Inventory when final stack empty in WID does not handle removing a custom ammo weapon with 0 ammo
  ;;[1 1 1] We check ammo in the GA this event triggers ( GA_DemonMask_Fire_AmmoTrackingPassiveFromEvent ) , but the client ammo count is not synced with server when we check to play montage and discard the weapon.
  ;; Passing through ammo as event magnitude lets us properly dispose of the weapon
  (if (not (CallFunction|IsAmmoInfinite))
    (Networking|SwitchHasAuthority
      (:Authority
        (Variables|Ammo|SetPrevious%20Local%20Mag%20Ammo _returnvalue)
        (if (<= (Math|Float|SelectFloat (Math|Conversions|ToFloat(Integer) (FortWeapon|GetBulletsPerClip)) (Math|Conversions|ToFloat(Integer) (FortWeapon|GetRemainingAmmo)) (FortWeapon|Ammo|HasInfiniteMagazineUses)) 0)
          (CallFunction|RemoveItemOnNoAmmoRemaining)))
      (:Remote
        (Variables|Ammo|SetPrevious%20Local%20Mag%20Ammo _returnvalue)))))

(event FortWeapon|EventOnLocalUnEquip
  ;[0 0.9688 0.1076 0.5] Check ammo remaining, if 0, discard the item. Persist in Inventory when final stack empty in WID does not handle removing a custom ammo weapon with 0 ammo
  (CallFunction|RemoveItemOnNoAmmoRemaining))

(event Custom|Remove%20Item%20on%20No%20Ammo%20Remaining
  ;[0 0.9688 0.1076 0.5] Check ammo remaining, if 0, discard the item. Persist in Inventory when final stack empty in WID does not handle removing a custom ammo weapon with 0 ammo
  (if (<= (FortWeapon|GetRemainingAmmo) 0)
    (bind _asfort_player_pawn (Utilities|Casting|CastToFortPlayerPawn (Game|GetInstigator)))
    ;;[1 1 1] Delay to avoid mantis crash and to sync with animation removal
    (Utilities|FlowControl|Delay 0.7)
    (Inventory|RemoveItemFromPlayerByGuid (Fortnite|GetFortPlayerControllerfromActor _asfort_player_pawn) (FortWeapon|GetInventoryGUID) 1 true)))

(event Custom|Cheat_DemonMaskFire_ForceImpactCue
  (params (FXVersion :type Integer))
  (GameplayCue|ExecuteGameplayCueOnActor(Burst) (Variables|Default|GetAs%20Fort%20Player%20Pawn) (Variables|Default|GetForce%20Impact%20Cue%20Gameplay%20Cue%20Tag) (Ability|GameplayCue|MakeGameplayCueParameters 0.0 (Math|Conversions|ToFloat(Integer) FXVersion))))

(event BPI_DemonMaskFireWeapon|EventSetWeaponProjectile
  (params (InProjectile :type B%20Prj%20Demon%20Mask%20Fire%20Object%20Reference))
  (Variables|Projectile|SetProjectile InProjectile))

(event BPI_DemonMaskFireWeapon|EventSetWeaponHomingTarget
  (params (InHomingTarget :type Actor%20Object%20Reference))
  (Variables|Projectile|SetHomingTarget InHomingTarget))

(fn ClearTimers
  (Utilities|Time|ClearandInvalidateTimerbyHandle (Variables|Reticle|GetReticleTickTimer))
  (Utilities|Time|ClearandInvalidateTimerbyHandle (Variables|Targeting|GetReticle%20Targeting%20Timer))
  (Utilities|Time|ClearandInvalidateTimerbyHandle (Variables|Reticle|GetTimer_ShowReticleStatus)))

(fn DisableVictimWarningOnEndStates
  ;[0.8281 0.3911 0 0.5] VictimWarning
  (if (not (Variables|TargetedVictim|GetDisableVictimWarning))
    (Variables|TargetedVictim|SetRepLockTarget 0)
    (Variables|TargetedVictim|SetLocalLockTarget 0)
    (bind _outmodifier (Class|WeaponFunctionLibrary|FindorAddCameraModifierforPawn (Variables|TargetedVictim|GetVictim%20Camera%20Modifier%20Class) (Variables|TargetedVictim|GetLocalLockTarget)))
    (bind _asdemon_mask_fire_camera_modifier_victim (Utilities|Casting|CastToDemonMask_Fire_CameraModifier_Victim _outmodifier))
    (Variables|/DemonMaskFire/Gameplay/Camera/DemonMask_Fire_CameraModifier_Victim.DemonMask_Fire_CameraModifier_Victim_C|SetIsActive false _asdemon_mask_fire_camera_modifier_victim)
    (Variables|TargetedVictim|SetWarningActive false)))

(fn GetAimedatTargetNonNativized
  (params (SourcePlayer :type Fort%20Player%20Controller%20Object%20Reference) (RectHalfWidth :type Float%20%28double-precision%29) (RectHalfHeight :type Float%20%28double-precision%29) (Range :type Float%20%28double-precision%29))
  (bind _sourceplayer (Variables|GetSourcePlayer))
  (if (Utilities|IsValid SourcePlayer :!class /Script/Engine.KismetSystemLibrary)
    (if (and (not (Pawn|IsLocalController SourcePlayer)) (Utilities|Enum|NotEqual(Enum) (Networking|GetLocalRole SourcePlayer) "ROLE_Authority"))
      (return 0)
      (else
        (LocalVariables|SetCameraPosition (Pawn|GetPlayerViewPoint _sourceplayer))
        (LocalVariables|SetCameraRotation (Pawn|GetBaseAimRotation (Pawn|GetControlledPawn _sourceplayer)))
        (LocalVariables|SetBestDistance (+ (Math|Float|Sqrt (+ (Math|Float|Square (Variables|GetRectHalfHeight)) (Math|Float|Square (Variables|GetRectHalfWidth)))) 0.0001))
        (LocalVariables|SetForwardAim (Math|Vector|GetForwardVector (LocalVariables|GetCameraRotation)))
        ;[0.4031 1 0.709 0.5] Find potential targets
        (bind (_outhits _returnvalue) (Collision|MultiBoxTraceForObjects (Utilities|Operators|Add :B (LocalVariables|GetCameraPosition) :C (* (LocalVariables|GetForwardAim) (Variables|GetRectHalfWidth))) (+ (LocalVariables|GetCameraPosition) (* (LocalVariables|GetForwardAim) (Variables|GetRange))) (Math|Vector|MakeVector (Variables|GetRectHalfWidth) (Variables|GetRectHalfHeight) 1e+01) (LocalVariables|GetCameraRotation) (Utilities|Array|MakeArray "ObjectTypeQuery3" "ObjectTypeQuery7") false (Variables|Targeting|GetActors%20to%20Ignore%20for%20Targeting) "None" true "(R=1.000000,G=0.000000,B=0.000000,A=1.000000)" "(R=0.000000,G=1.000000,B=0.000000,A=1.000000)" 5.01))
        (if _returnvalue
          (LocalVariables|SetLocal_OutHits _outhits)
          (for _array_element (LocalVariables|GetLocal_OutHits)
            (bind _ascharacter (Utilities|Casting|CastToCharacter (.HitActor (Collision|BreakHitResult _array_element))))
            (if (GameplayTags|HasAnyMatchingGameplayTags (Utilities|Casting|ToGameplayTagAssetInterface _ascharacter) (Variables|Default|GetAvoid%20Detection%20Tags%20Container))
              _
              (elif (not (Team|OnSameTeam (Variables|GetSourcePlayer) _ascharacter))
                ;[0.6979 0.7664 1 0.5] Is this target closer to our reticle than the current best candidate?
                (bind _output_get (LocalVariables|SetPotentialDistance (Math|Vector|GetPointDistancetoLine (Transformation|GetActorLocation (.HitActor (Collision|BreakHitResult _array_element))) (LocalVariables|GetCameraPosition) (LocalVariables|GetForwardAim))))
                (if (< _output_get (LocalVariables|GetBestDistance))
                  (bind _returnvalue_2 (Fortnite|CheckLineOfSightToActorWithChannel (LocalVariables|GetCameraPosition) (.HitActor (Collision|BreakHitResult _array_element)) "ECC_GameTraceChannel5" (Variables|GetSourcePlayer))) ; Check if there's anything in the way, like a wall.\nOnly do this raycast if the other checks have been passed.
                  (if _returnvalue_2
                    (LocalVariables|SetBestTarget (.HitActor (Collision|BreakHitResult _array_element)))
                    (LocalVariables|SetBestDistance (LocalVariables|GetPotentialDistance)))))))
          (return (LocalVariables|GetBestTarget)))))
    (else
      (return 0))))

(fn GetDemonMaskFireTargetInfoNonNativized
  (params (Source%20Player%20Pawn :type Fort%20Player%20Pawn%20Object%20Reference) (Draw%20Debug%20Duration :type Float%20%28double-precision%29))
  (bind _returnvalue (Transformation|GetActorLocation Source%20Player%20Pawn))
  (bind _source_player_pawn (Variables|GetSource%20Player%20Pawn))
  (bind _returnvalue_2 (Fortnite|GetFortPlayerControllerfromActor _source_player_pawn))
  (bind _returnvalue_3 (Reticle|GetActorUnderReticle _returnvalue_2))
  (bind _returnvalue_4 (Fortnite|GetFortPlayerControllerfromActor Source%20Player%20Pawn))
  (bind _returnvalue_5 (Reticle|GetLocationUnderReticle _returnvalue_4))
  (bind _returnvalue_6 (- _returnvalue_5 _returnvalue))
  (bind _returnvalue_7 (Math|Vector|Normalize _returnvalue_6))
  (bind (_asfort_player_pawn _bsuccess) (Utilities|Casting|CastToFortPlayerPawn _returnvalue_3))
  (bind _returnvalue_8 (* _returnvalue_7 2e+03))
  (bind _returnvalue_9 (+ _returnvalue_8 _returnvalue))
  (bind _draw_debug_duration (Variables|GetDraw%20Debug%20Duration))
  (Variables|Targeting|SetActors%20to%20Ignore%20for%20Targeting (Utilities|Array|MakeArray Source%20Player%20Pawn))
  (if (and (Pawn|IsLocallyViewed Source%20Player%20Pawn) (Utilities|IsValid Source%20Player%20Pawn :!class /Script/Engine.KismetSystemLibrary))
    (if (> (Variables|GetDraw%20Debug%20Duration) 0)
      (LocalVariables|SetDrawDebugType "ForDuration"))
    (bind (_outhit _returnvalue_10) (Collision|SphereTraceForObjects (+ (* _returnvalue_7 3e+02) _returnvalue) _returnvalue_9 3e+02 (LocalVariables|GetPawnObjectTypes) false (Variables|Targeting|GetActors%20to%20Ignore%20for%20Targeting) (LocalVariables|GetDrawDebugType) true "(R=1.000000,G=0.000000,B=0.000000,A=1.000000)" "(R=0.000000,G=1.000000,B=0.000000,A=1.000000)" _draw_debug_duration))
    (if _returnvalue_10
      (bind _success (CallFunction|TestforLineOfSight (Variables|GetSource%20Player%20Pawn) (.HitActor (Collision|BreakHitResult _outhit))))
      (if _success
        (return (Collision|BreakHitResult _outhit) (.HitActor (Collision|BreakHitResult _outhit)) (.HitComponent (Collision|BreakHitResult _outhit)) (.ImpactPoint (Collision|BreakHitResult _outhit)) (.Distance (Collision|BreakHitResult _outhit)))
        (else
          _))
      (else
        (bind (_outhit_2 _returnvalue_11) (Collision|SphereTraceForObjects _returnvalue_9 (+ (* _returnvalue_7 6e+03) _returnvalue) 3e+02 (LocalVariables|GetPawnObjectTypes) false (Variables|Targeting|GetActors%20to%20Ignore%20for%20Targeting) (LocalVariables|GetDrawDebugType) true "(R=1.000000,G=0.000000,B=0.000000,A=1.000000)" "(R=0.000000,G=1.000000,B=0.000000,A=1.000000)" _draw_debug_duration))
        (if _returnvalue_11
          (bind _success_2 (CallFunction|TestforLineOfSight (Variables|GetSource%20Player%20Pawn) (.HitActor (Collision|BreakHitResult _outhit_2))))
          (if _success_2
            (return (Collision|BreakHitResult _outhit_2) (.HitActor (Collision|BreakHitResult _outhit_2)) (.HitComponent (Collision|BreakHitResult _outhit_2)) (.ImpactPoint (Collision|BreakHitResult _outhit_2)) (.Distance (Collision|BreakHitResult _outhit_2)))
            (else
              _)))))
    (if (and _bsuccess (<= (Math|Vector|Distance(Vector) (Transformation|GetActorLocation (Game|GetInstigator _source_player_pawn)) (Reticle|GetLocationUnderReticle (Fortnite|GetFortPlayerControllerfromActor _source_player_pawn))) 6e+03))
      (return true _asfort_player_pawn 0 "0, 0, 0" 0.0)
      (else
        (return false 0 0 "0, 0, 0" 0.0)))
    (else
      (return false 0 0 "0, 0, 0" 0.0))))

(fn GetWeaponHomingTarget
  (return (Variables|Projectile|GetHomingTarget)))

(fn GetWeaponMaxProjectileSpeed
  (return (Variables|Projectile|GetPrjMaxSpeed)))

(fn IsAmmoInfinite
  (return (or (FortWeapon|Ammo|HasInfiniteCharges) (FortWeapon|Ammo|InventoryOwnerHasInfiniteUses))))

(fn OnRep_Projectile)

(fn OnRep_RepLockTarget
  (bind _replocktarget (Variables|TargetedVictim|GetRepLockTarget))
  (if (Networking|IsServer)
    _
    (elif (Variables|TargetedVictim|GetWarningActive)
      (Utilities|IsValid _replocktarget
        (:Is%20Valid
          (if (!= (Variables|TargetedVictim|GetLocalLockTarget) _replocktarget)
            _))
        (:Is%20Not%20Valid
          _))
      (bind _outmodifier (Class|WeaponFunctionLibrary|FindorAddCameraModifierforPawn (Variables|TargetedVictim|GetVictim%20Camera%20Modifier%20Class) (Variables|TargetedVictim|GetLocalLockTarget)))
      (bind _asdemon_mask_fire_camera_modifier_victim (Utilities|Casting|CastToDemonMask_Fire_CameraModifier_Victim _outmodifier))
      (Variables|/DemonMaskFire/Gameplay/Camera/DemonMask_Fire_CameraModifier_Victim.DemonMask_Fire_CameraModifier_Victim_C|SetIsActive false _asdemon_mask_fire_camera_modifier_victim)
      (Variables|TargetedVictim|SetWarningActive false)
      (else
        (Utilities|IsValid _replocktarget
          (:Is%20Valid
            (if (Pawn|IsLocallyViewed _replocktarget)
              (Variables|TargetedVictim|SetWarningActive true)
              (Variables|TargetedVictim|SetLocalLockTarget _replocktarget)
              (bind _outmodifier_2 (Class|WeaponFunctionLibrary|FindorAddCameraModifierforPawn (Variables|TargetedVictim|GetVictim%20Camera%20Modifier%20Class) _replocktarget))
              (bind _asdemon_mask_fire_camera_modifier_victim_2 (Utilities|Casting|CastToDemonMask_Fire_CameraModifier_Victim _outmodifier_2))
              (Variables|/DemonMaskFire/Gameplay/Camera/DemonMask_Fire_CameraModifier_Victim.DemonMask_Fire_CameraModifier_Victim_C|SetIsActive true _asdemon_mask_fire_camera_modifier_victim_2))))))))

(fn OnRep_TotalAmmoCount)

(fn ReticleUpdate
  ;[1 1 1] Check if we have objects we can hit in homing range. If player, update that bool for UI specific case
  (Variables|Projectile|GetProjectileReplicated
    (:then
      _)
    (:else
      (bind _asfort_player_pawn (Utilities|Casting|CastToFortPlayerPawn (Game|GetInstigator)))
      (if (Utilities|IsValid _asfort_player_pawn :!class /Script/Engine.KismetSystemLibrary)
        (LocalVariables|SetEnvTargetInRange (<= (Math|Vector|Distance(Vector) (Transformation|GetActorLocation (Game|GetInstigator)) (Reticle|GetLocationUnderReticle (Fortnite|GetFortPlayerControllerfromActor _asfort_player_pawn))) (ScalableFloat|GetValueatLevel (Variables|Hotfix|GetDataHomingRange))))
        (LocalVariables|SetPawnTargetInRange (.bSuccess (Utilities|Casting|CastToFortPlayerPawn (Reticle|GetActorUnderReticle (Fortnite|GetFortPlayerControllerfromActor _asfort_player_pawn))))))))
  (Default|CallUpdateReticleState (LocalVariables|GetPawnTargetInRange) (LocalVariables|GetEnvTargetInRange)))

(fn SetTargetHomingActor
  (params (New%20Target :type Actor%20Object%20Reference))
  (if (!= New%20Target (Variables|Targeting|GetTargetHomingActor))
    (bind _output_get (Variables|Targeting|SetTargetHomingActor New%20Target))
    (CallFunction|EventServerSetTargetHomingActor _output_get)))

(fn TestForLineOfSight
  (params (Source%20Player%20Pawn :type Fort%20Player%20Pawn%20Object%20Reference) (Target%20Actor :type Actor%20Object%20Reference))
  (if (and (Pawn|IsLocallyViewed Source%20Player%20Pawn) (Utilities|IsValid Source%20Player%20Pawn :!class /Script/Engine.KismetSystemLibrary))
    (bind _outhit (Collision|LineTraceByChannel (Transformation|GetActorLocation Source%20Player%20Pawn) (Math|Transform|TransformLocation (Transformation|GetActorTransform Source%20Player%20Pawn) (* (Math|Vector|VectorUp) 5e+01))))
    (return (and (== (.HitActor (Collision|BreakHitResult _outhit)) Target%20Actor) (Collision|BreakHitResult _outhit)))))

(fn TickVictimWarning
  (bind _replocktarget (Variables|TargetedVictim|GetRepLockTarget))
  (if (Utilities|IsValid (Variables|TargetedVictim|GetRepLockTarget) :!class /Script/Engine.KismetSystemLibrary)
    (Utilities|IsValid _replocktarget
      (:Is%20Valid
        (if (or (FortPawn|IsDead _replocktarget) (Health|IsDBNO _replocktarget))
          (Variables|TargetedVictim|SetWarningActive false)
          (bind _outmodifier (Class|WeaponFunctionLibrary|FindorAddCameraModifierforPawn (Variables|TargetedVictim|GetVictim%20Camera%20Modifier%20Class) (Variables|TargetedVictim|GetLocalLockTarget)))
          (bind _asdemon_mask_fire_camera_modifier_victim (Utilities|Casting|CastToDemonMask_Fire_CameraModifier_Victim _outmodifier))
          (Variables|/DemonMaskFire/Gameplay/Camera/DemonMask_Fire_CameraModifier_Victim.DemonMask_Fire_CameraModifier_Victim_C|SetIsActive false _asdemon_mask_fire_camera_modifier_victim)))
      (:Is%20Not%20Valid
        (Variables|TargetedVictim|SetWarningActive false)))))

(fn UpdateReticleTargeting
  (bind _asfort_player_pawn (Utilities|Casting|CastToFortPlayerPawn (Game|GetInstigator)))
  (bind _output_get (LocalVariables|SetLocal%20Player%20Pawn _asfort_player_pawn))
  (Utilities|IsValid _output_get
    (:Is%20Valid
      (bind _returnvalue (CallFunction|GetAimedatTarget (Fortnite|GetFortPlayerControllerfromActor _output_get) (ScalableFloat|GetValueatLevel (Variables|Hotfix|GetDataHomingRectHeightandWidth)) (ScalableFloat|GetValueatLevel (Variables|Hotfix|GetDataHomingRectHeightandWidth)) (ScalableFloat|GetValueatLevel (Variables|Hotfix|GetDataHomingRange)) (Utilities|Array|MakeArray "ObjectTypeQuery3" "ObjectTypeQuery7") (Variables|Targeting|GetActors%20to%20Ignore%20for%20Targeting) (Variables|Default|GetAvoid%20Detection%20Tags%20Container)))
      (Utilities|IsValid _returnvalue
        (:Is%20Valid
          (Variables|Targeting|SetReticleWorldLocation (Math|Transform|TransformLocation (Transformation|GetActorTransform _returnvalue) (* (Math|Vector|VectorUp) (Math|Float|SelectFloat 15.0 5e+01 (Health|IsDBNO (Utilities|Casting|CastToFortPlayerPawn _returnvalue)))))))
        (:Is%20Not%20Valid
          (Variables|Targeting|SetReticleWorldLocation "0, 0, 0")))
      (Utilities|Casting|CastToFortPlayerPawn _returnvalue
        (:then
          _)
        (:CastFailed
          (Utilities|Casting|CastToFortAIPawn _returnvalue
            (:then
              (Variables|Targeting|SetReticleState "NewEnumerator7")
              (Default|SetTargetHomingActor _returnvalue))
            (:CastFailed
              (Variables|Targeting|SetReticleState "NewEnumerator0")
              (Default|SetTargetHomingActor 0)))
          (Class|FortWeaponComponentTargetState|UpdateTargetInfo (Variables|Default|GetFortWeaponComponent_TargetState) (Utilities|Enum|NotEqual(Enum) (Variables|Targeting|GetReticleState) "NewEnumerator0") (Variables|Targeting|GetReticleWorldLocation))
          (Development|PrintString (Utilities|String|Append "Demon Fire Mask Reticle: " (Utilities|String|EnumtoString (Variables|Targeting|GetReticleState)))))))))

(fn UpdateReticleVisibility
  (bind _projectilereplicated (Variables|Projectile|GetProjectileReplicated))
  (bind _returnvalue (Utilities|IsValid _projectilereplicated :!class /Script/Engine.KismetSystemLibrary))
  (bind _isinthrowablestate (Variables|Projectile|GetIsInThrowableState))
  (bind _returnvalue_2 (CallFunction|IsBeingEquipped))
  (bind _returnvalue_3 (FortWeapon|IsEquipped))
  (bind _returnvalue_4 (Math|Boolean|NORBoolean _returnvalue_3 _returnvalue_2))
  (bind _returnvalue_5 (Math|Boolean|ORBoolean _returnvalue_4 _isinthrowablestate _returnvalue))
  (bind _returnvalue_6 (not _returnvalue_5))
  (bind _hidereticletags (Variables|Reticle|GetHideReticleTags))
  (Networking|CanExecuteCosmeticEvents
    (:True
      (bind _as_fort_player_pawn (Variables|Default|GetAs%20Fort%20Player%20Pawn)
        (:then
          (if (Pawn|IsLocallyControlled _as_fort_player_pawn)
            (HUD|SetHUDElementVisibility (BlueprintContext|GetFortHUDContext) (Variables|Reticle|GetReticle%20HUD%20Element%20Tag) (or (GameplayTags|HasAnyMatchingGameplayTags _as_fort_player_pawn _hidereticletags) _returnvalue_6) "None") ; Show the reticle when in the throw state and when the projectile is currently flying
            (if (or (GameplayTags|HasAnyMatchingGameplayTags _as_fort_player_pawn _hidereticletags) _returnvalue_6)
              (Utilities|Time|ClearandInvalidateTimerbyHandle (Variables|Reticle|GetReticleTickTimer))
              (else
                (bind _returnvalue_7 (Utilities|Time|SetTimerbyEvent (EventDispatchers|CreateEvent :!event ReticleTick) 0.01 true))
                (Variables|Reticle|SetReticleTickTimer _returnvalue_7)))))
        (:else)))))

(fn ConstructionScript
  (|CallParentFunction))
