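; Demon Mask fire weapon projectile: a fireball that scales up on spawn, seeks targets and explodes before its lifetime ends or on touching water.
; Hotfix rows drive its tuning, and timers delay capsule growth, start velocity-multiplier acceleration with a gameplay cue, and poll for targets.
; The homing target is stored on the owning weapon to avoid a BeginPlay race, then applied as authority-side lock-on homing.
; Lifetime explosions apply a gameplay effect through async targeting.
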(event Custom|PlayFireballShootSound
  (Audio|SpawnSoundAttached (Variables|Audio|GetFireballShootSound) (Variables|Default|GetMesh)))

(event EventBeginPlay
  (Variables|Gameplay|SetbUseAsyncExplosions (ScalableFloat|GetValueasBool (Variables|Default|GetHF_UseAsyncExplosion)))
  (|CallParentFunction)
  (bind _asfort_player_pawn (Utilities|Casting|CastToFortPlayerPawn (Game|GetInstigator))
    (:then
      (Variables|Default|SetFort%20Player%20Pawn _asfort_player_pawn))
    (:CastFailed
      _))
  (CallFunction|SetProjectileVaraiblesfromHotfixRows)
  (Utilities|FlowControl|Sequence
    (:then_0
      ; Water Interact Explode Delay Timer
      (Utilities|Time|SetTimerbyEvent (EventDispatchers|CreateEvent :!event Custom|WaterExplodeDelay) (Variables|Default|GetWaterInteractionDelay))
      (Utilities|Time|SetTimerbyEvent (EventDispatchers|CreateEvent :!event Custom|ExplodeBeforeLifetimeExpires) (Math|Float|Clamp(Float) (Variables|Actor|GetInitialLifeSpan) 0.1 (- (Variables|Actor|GetInitialLifeSpan) 0.5)))
      (bind (_direction _newtrack_0) (Timeline|ScaleMeshTimeline PlayFromStart)
        (:Update
          (Transformation|SetWorldScale3D (Variables|Default|GetMesh) (Math|Conversions|ToVector(Float) _newtrack_0))
          (Transformation|SetWorldScale3D (Variables|Default|GetNiagara_trail_SwitchToGCNL) (Math|Conversions|ToVector(Float) _newtrack_0)))))
    (:then_1
      ;[0.232 0.9323 0.7054 0.5] Init Player Weapon Specific Data
      (bind _returnvalue (CallFunction|GetOwnerWeapon))
      (bind _asbpi_demon_mask_fire_weapon (Utilities|Casting|CastToBPI_DemonMaskFireWeapon _returnvalue)
        (:then
          (Variables|Default|SetDemonMaskFireOwnerWeapon _returnvalue)
          (BPI_DemonMaskFireWeapon|SetWeaponProjectile _asbpi_demon_mask_fire_weapon self))
        (:CastFailed
          (Development|PrintString "FAIL"))))
    (:then_3
      (bind _returnvalue_2 (GameplayTags|HasAnyMatchingGameplayTags(Message) (Variables|Default|GetFort%20Player%20Pawn) (Variables|Default|GetOn%20Train%20Tag%20Container)))
      (bind _returnvalue_3 (Utilities|Time|SetTimerbyEvent (EventDispatchers|CreateEvent :!event Custom|WaitThenSetProjectileSpeedMax) (ScalableFloat|GetValueatLevel (select _returnvalue_2 (Variables|Speed|GetRow_BasedMovingDelayBeforeAccel) (Variables|Homing|GetRow_DelayBeforeSpeedIncrease)))))
      (Variables|Default|SetDelaySpeedIncreaseTimer _returnvalue_3)
      (Utilities|FlowControl|Delay 0.1)
      (if (Math|Boolean|ANDBoolean (not (Game|IsActorBeingDestroyed)) (Utilities|IsValid self :!class /Script/Engine.KismetSystemLibrary) (not (Delivery|IsBeingKilled)))
        ;[0 0.5896 0.8594 0.5] Allow Homing
        (if (ScalableFloat|GetValueasBool (Variables|Homing|GetRow_HomingEnabled))
          (Default|SetHomingVars)
          (bind _returnvalue_4 (Utilities|Time|SetTimerbyEvent (EventDispatchers|CreateEvent :!event PeriodicCheckForHomingTarget) (ScalableFloat|GetValueatLevel (Variables|Homing|GetRow_HomingCheckInterval)) true))
          (Variables|Homing|SetCheckForHomingTargetTimer _returnvalue_4)
          (else
            _))
        ;[0 0.1244 0.9635 0.5] Delay then increase capsule size. This is to help prevent accidental hits when firing near close objects
        (bind _returnvalue_5 (Utilities|Time|SetTimerbyEvent (delegate Custom|DemonMaskFire_SetCapsuleSize_AfterDelay) (ScalableFloat|GetValueatLevel (Variables|Default|GetHF_DelayBeforeCapsuleIncrease))))
        (Variables|Default|SetCapsuleSizeTimerHandle _returnvalue_5)))))

(event Custom|DebugSpeedEvent
  (Development|PrintString (Utilities|String|ToString(Float) (Variables|/Script/Engine.ProjectileMovementComponent|GetMaxSpeed (Variables|Movement|GetProjectileMovementComponent))) true true "(R=0.000000,G=1.000000,B=0.174587,A=1.000000)"))

(event Custom|DemonMaskFire_SetCapsuleSize_AfterDelay
  ;[0 0.1244 0.9635 0.5] Delay then increase capsule size. This is to help prevent accidental hits when firing near close objects
  (Components|Capsule|SetCapsuleSize (Variables|Collision|GetCapsuleComponent) (ScalableFloat|GetValueatLevel (Variables|Default|GetHF_CapsuleSizeAfterDelay)) (ScalableFloat|GetValueatLevel (Variables|Default|GetHF_CapsuleSizeAfterDelay)) true))

(event Collision|EventHit
  (params (MyComp :type Primitive%20Component%20Object%20Reference) (Other :type Actor%20Object%20Reference) (OtherComp :type Primitive%20Component%20Object%20Reference) (bSelfMoved :type Boolean) (HitLocation :type Vector :default "0, 0, 0") (HitNormal :type Vector :default "0, 0, 0") (NormalImpulse :type Vector :default "0, 0, 0") (Hit :type Hit%20Result%20Structure%20%28by%20ref%29))
  (bind _returnvalue (Game|GetPlayerCharacter 0))
  ;[0.05697 0.09873 0.15 0.5] If impact is close enough to player, toggle audio control bus mix
  (Collision|SetCollisionEnabled (Variables|Collision|GetCapsuleComponent) "NoCollision")
  (bind _asfort_player_pawn_athena (Utilities|Casting|CastToFortPlayerPawnAthena _returnvalue))
  (if (and (Pawn|IsLocallyViewed _asfort_player_pawn_athena) (<= (Math|Vector|VectorLength (- HitLocation (Transformation|GetWorldLocation (Variables|/Script/Engine.Character|GetMesh _returnvalue)))) 185.0))
    (Variables|Default|SetbHitIsValid true)))

(event Custom|OnImpactControlBusMix
  ;[0.05697 0.09873 0.15 0.5] If impact is close enough to player, toggle audio control bus mix
  (if (Variables|Default|GetbHitIsValid)
    _))

(event Custom|WaitThenSetProjectileSpeedMax
  (Variables|/Script/Engine.ProjectileMovementComponent|SetMaxSpeed (CallFunction|GetMaxSpeed) (Variables|Movement|GetProjectileMovementComponent))
  ;[0.9948 0.8359 0.01036 0.5] TODO: Updating max speed without adjusting velocity doesn't work. This is mainly for non-homing cases.
  ; Fake a curve for now
  (bind _returnvalue (Utilities|Time|SetTimerbyEvent (EventDispatchers|CreateEvent :!event Custom|AccelerateMaskOverTime) 0.05 true true))
  (Variables|Acceleration|SetAccelerateMovementTimerHandle _returnvalue)
  (bind _returnvalue_2 (Utilities|Time|SetTimerbyEvent (EventDispatchers|CreateEvent :!event Custom|AccelGC) (ScalableFloat|GetValueatLevel (Variables|Acceleration|GetAccelGCDelay))))
  (Variables|Acceleration|SetAccelGCTimerHandle _returnvalue_2))

(event Custom|AccelGC
  ;[0.9948 0.8359 0.01036 0.5] TODO: Updating max speed without adjusting velocity doesn't work. This is mainly for non-homing cases.
  ; Fake a curve for now
  (if (not (Math|Boolean|ORBoolean (Game|IsActorBeingDestroyed self) (Delivery|IsBeingKilled self) (Projectile|IsProjectileBeingKilled self)))
    (GameplayCues|ExecuteGameplayCue self (Variables|Acceleration|GetAccelerate%20Gameplay%20Cue%20Tag))))

(event Custom|AccelerateMaskOverTime
  (bind _initialaccelspeedmulti (Variables|Speed|GetInitialAccelSpeedMulti))
  (bind _velocityincreaseamount (Variables|Acceleration|GetVelocityIncreaseAmount))
  (bind _returnvalue (+ _velocityincreaseamount _initialaccelspeedmulti))
  (bind _returnvalue_2 (> _returnvalue 1.15))
  (bind _increasedaccelspeedmulti (Variables|Speed|GetIncreasedAccelSpeedMulti))
  (bind _velocityincreaseamount_2 (Variables|Acceleration|GetVelocityIncreaseAmount))
  (bind _returnvalue_3 (+ _velocityincreaseamount_2 _increasedaccelspeedmulti))
  (bind _returnvalue_4 (Math|Float|SelectFloat _returnvalue_3 _returnvalue _returnvalue_2))
  (bind _returnvalue_5 (Math|Float|Clamp(Float) _returnvalue_4 1.0 2.0))
  (bind _projectilemovementcomponent (Variables|Movement|GetProjectileMovementComponent))
  (bind _velocity (Variables|/Script/Engine.MovementComponent|GetVelocity _projectilemovementcomponent))
  (bind _returnvalue_6 (* _velocity _returnvalue_5))
  (bind _returnvalue_7 (Game|Components|ProjectileMovement|LimitVelocity _projectilemovementcomponent _returnvalue_6))
  ;[0.9948 0.8359 0.01036 0.5] TODO: Updating max speed without adjusting velocity doesn't work. This is mainly for non-homing cases.
  ; Fake a curve for now
  (Variables|Homing|GetHomingTarget
    (:then
      _)
    (:else
      _))
  (Variables|Acceleration|SetVelocityIncreaseAmount _returnvalue_5)
  (if (Utilities|Operators|Equal(==) _velocity _returnvalue_7 0.1)
    (Utilities|Time|ClearandInvalidateTimerbyHandle (Variables|Acceleration|GetAccelerateMovementTimerHandle))
    (else
      (Projectile|SetVelocity _projectilemovementcomponent _returnvalue_7))))

(event OnProjectileStop(ProjectileMovementComponent)
  (params (ImpactResult :type Hit%20Result%20Structure%20%28by%20ref%29))
  (bind _cbm_fireballimpact (Variables|Audio|GetCBM_FireballImpact)
    (:then
      (if (Audio|Modulation|IsControlBusMixActive _cbm_fireballimpact)
        (Audio|Modulation|DeactivateControlBusMix _cbm_fireballimpact)))
    (:else)))

(event EventEndPlay
  (params (EndPlayReason :type EEndPlayReason%20Enum :default "Destroyed"))
  (|CallParentFunction EndPlayReason)
  (Utilities|Time|ClearandInvalidateTimerbyHandle (Variables|Default|GetCapsuleSizeTimerHandle)))

(event Custom|ExplodeBeforeLifetimeExpires
  (if (not (Variables|Default|GetAlreadyExploded))
    (GameplayCues|ExecuteGameplayCue (Variables|Default|GetExploded%20Gameplay%20Cue%20Tag))
    (bind _fort_player_pawn (Variables|Default|GetFort%20Player%20Pawn)
      (:then
        (bind (_asynctaskproxy _targetinghandle) (Class|AsyncActionPerformTargeting|PerformTargetingAsyncAction 0 "/DemonMaskFire/Gameplay/Targeting/TP_DemonMask_Fire_Explosion.TP_DemonMask_Fire_Explosion" true)
          (:Targeted
            (Ability|Tasks|ApplyGameplayEffectToTargetingHandleAsync (Ability|GetAbilitySystemComponent _fort_player_pawn) _targetinghandle (Variables|Default|GetLifetime%20Exploded%20Gameplay%20Effect%20Class) 1 1)
            (Delivery|DoExplosion)
            (Variables|Default|SetAlreadyExploded true))))
      (:else))))

(event EventOnTouched
  (params (OtherActor :type Actor%20Object%20Reference) (OtherComp :type Primitive%20Component%20Object%20Reference) (HitResult :type Hit%20Result%20Structure%20%28by%20ref%29) (bIsOverlap :type Boolean))
  (bind _capsulecomponent (Variables|Collision|GetCapsuleComponent))
  (bind _returnvalue (Transformation|GetWorldLocation _capsulecomponent))
  (|CallParentFunction OtherActor OtherComp HitResult bIsOverlap)
  (if (not (Variables|Default|GetAlreadyExploded))
    (Variables|Default|SetAlreadyExploded true)
    (bind _aswater_body (Utilities|Casting|CastToWaterBody OtherActor)
      (:then
        (Utilities|FlowControl|Sequence
          (:then_0
            _)
          (:then_1
            ;[1 1 1] Handle Water
            (Variables|Default|SetWater%20Interaction%20Component (stale OnEnterWaterEvent(WaterInteractionComponent) WaterInteractionComponent))
            (bind _output_get (Variables|Default|SetOut%20Water%20Surface%20Location (WaterBody|GetWaterSurfaceInfoatLocation (Water|GetWaterBodyComponent (select (Utilities|IsValid _aswater_body :!class /Script/Engine.KismetSystemLibrary) _aswater_body (stale OnEnterWaterEvent(WaterInteractionComponent) _output_get))) _returnvalue false)))
            (Variables|Default|SetHitWater true)
            (Effects|Components|ParticleSystem|SpawnFXAtLocation (Variables|Default|GetWaterImpactFX) _output_get (Math|Rotator|MakeRotfromX (.OutWaterSurfaceNormal (WaterBody|GetWaterSurfaceInfoatLocation (Water|GetWaterBodyComponent (select (Utilities|IsValid _aswater_body :!class /Script/Engine.KismetSystemLibrary) _aswater_body (stale OnEnterWaterEvent(WaterInteractionComponent) _output_get))) _returnvalue false))) "2.000000,2.000000,2.000000")
            (CallFunction|ExplodeBeforeLifetimeExpires))))
      (:CastFailed
        _))
    (Variables|Projectile|SetbStopSimulatingOnHit true)
    (Utilities|Time|ClearandInvalidateTimerbyHandle (Variables|Acceleration|GetAccelGCTimerHandle))
    (Utilities|Time|ClearandInvalidateTimerbyHandle (Variables|Default|GetCapsuleSizeTimerHandle))
    (Utilities|Time|ClearandInvalidateTimerbyHandle (Variables|Acceleration|GetAccelerateMovementTimerHandle))))

(event OnEnterWaterEvent(WaterInteractionComponent)
  (params (WaterBody :type Fort%20Water%20Body%20Actor%20Object%20Reference) (WaterInteractionComponent :type Fort%20Water%20Interaction%20Component%20Object%20Reference) (bIsFirstBody :type Boolean))
  (bind _capsulecomponent (Variables|Collision|GetCapsuleComponent))
  (bind _returnvalue (Transformation|GetWorldLocation _capsulecomponent))
  ;[1 1 1] Handle Water
  (if bIsFirstBody
    (bind _output_get (Variables|Default|SetWater%20Body WaterBody))
    (Variables|Default|SetWater%20Interaction%20Component WaterInteractionComponent)
    (bind _output_get_2 (Variables|Default|SetOut%20Water%20Surface%20Location (WaterBody|GetWaterSurfaceInfoatLocation (Water|GetWaterBodyComponent (select (Utilities|IsValid (stale EventOnTouched _aswater_body) :!class /Script/Engine.KismetSystemLibrary) (stale EventOnTouched _aswater_body) _output_get)) _returnvalue false)))
    (Variables|Default|SetHitWater true)
    (Effects|Components|ParticleSystem|SpawnFXAtLocation (Variables|Default|GetWaterImpactFX) _output_get_2 (Math|Rotator|MakeRotfromX (.OutWaterSurfaceNormal (WaterBody|GetWaterSurfaceInfoatLocation (Water|GetWaterBodyComponent (select (Utilities|IsValid (stale EventOnTouched _aswater_body) :!class /Script/Engine.KismetSystemLibrary) (stale EventOnTouched _aswater_body) _output_get)) _returnvalue false))) "2.000000,2.000000,2.000000")
    (CallFunction|ExplodeBeforeLifetimeExpires)))

(event EventDestroyed
  (bind _asbpi_demon_mask_fire_weapon (Utilities|Casting|CastToBPI_DemonMaskFireWeapon (Variables|Default|GetDemonMaskFireOwnerWeapon))
    (:then
      (BPI_DemonMaskFireWeapon|SetWeaponHomingTarget _asbpi_demon_mask_fire_weapon 0))
    (:CastFailed
      _))
  (Variables|Homing|SetHomingTarget 0)
  (Utilities|Time|ClearandInvalidateTimerbyHandle (Variables|Default|GetCapsuleSizeTimerHandle))
  (Utilities|Time|ClearandInvalidateTimerbyHandle (Variables|Acceleration|GetAccelerateMovementTimerHandle))
  (Utilities|Time|ClearandInvalidateTimerbyHandle (Variables|Acceleration|GetAccelGCTimerHandle)))

(event Custom|WaterExplodeDelay
  (bind _projectilemovementcomponent (Variables|Movement|GetProjectileMovementComponent))
  ;[1 1 1] Handle Water
  (Collision|SetCollisionResponsetoChannel (Variables|/Script/Engine.MovementComponent|GetUpdatedPrimitive _projectilemovementcomponent) "ECC_GameTraceChannel18" "ECR_Block")
  (Collision|SetCollisionResponsetoChannel (Variables|Collision|GetCapsuleComponent) "ECC_GameTraceChannel18" "ECR_Block")
  (if (and (Utilities|IsValid (Variables|/Script/Engine.MovementComponent|GetUpdatedComponent _projectilemovementcomponent) :!class /Script/Engine.KismetSystemLibrary) (not (Delivery|IsBeingKilled)))
    (Variables|Projectile|SetTouchWaterBehavior "StopIfStopSimulatingOnHit")
    (if (Water|IsinWater (Variables|Water|GetWaterInteractionComponent))
      (if (not (Delivery|IsBeingKilled))
        (Delivery|DoExplosion)))))

(fn FindandSetHomingTarget
  ;[0 0.5834 0.9635 0.5] Recreation of GetAimedAtEnemy(), but allowing non-FortPawns to be returned. Includes AI Pawns by default.
  ; The previous GetAimedAtEnemy() node had "Target AI Pawns" true at all times.
  (bind _returnvalue (CallFunction|GetAimedatTarget (Fortnite|GetFortPlayerControllerfromActor (Actor|GetOwner)) (Math|Float|SelectFloat (Variables|Homing|GetHomingWidthAndHeight) (Variables|Homing|GetSeekingHomingWidthAndHeight) (Variables|Default|GetInitialCheckForTarget)) (Math|Float|SelectFloat (Variables|Homing|GetHomingWidthAndHeight) (Variables|Homing|GetSeekingHomingWidthAndHeight) (Variables|Default|GetInitialCheckForTarget)) (Variables|Homing|GetHomingRange) (Utilities|Array|MakeArray "ObjectTypeQuery3" "ObjectTypeQuery7") (Variables|Homing|GetActorsToIgnoreForTargeting)))
  (Variables|Default|SetInitialCheckForTarget false)
  (Utilities|IsValid _returnvalue
    (:Is%20Valid
      (if (CallFunction|IsValidHomingTarget _returnvalue)
        ;[1 1 1] Store Homing target on Weapon, projectile can pull from there. Trying to set directly on projectile after spawning it has race condition issues with projectile Begin Play
        (CallFunction|UpdateHomingTarget _returnvalue)
        (Default|SetHomingVars)
        (else
          (CallFunction|UpdateHomingTarget 0))))))

(fn GetHomingTarget
  (bind _asbpi_demon_mask_fire_weapon (Utilities|Casting|CastToBPI_DemonMaskFireWeapon (Variables|Default|GetDemonMaskFireOwnerWeapon)))
  (bind _homingtarget (BPI_DemonMaskFireWeapon|GetWeaponHomingTarget _asbpi_demon_mask_fire_weapon))
  (return _homingtarget))

(fn GetMaxSpeed
  (bind _asbpi_demon_mask_fire_weapon (Utilities|Casting|CastToBPI_DemonMaskFireWeapon (Variables|Default|GetDemonMaskFireOwnerWeapon)))
  (bind _maxspeed (BPI_DemonMaskFireWeapon|GetWeaponMaxProjectileSpeed _asbpi_demon_mask_fire_weapon))
  (return _maxspeed))

(fn IsValidHomingTarget
  (params (Target :type Actor%20Object%20Reference))
  (bind _avoid_detection_tags_container (Variables|Homing|GetAvoid%20Detection%20Tags%20Container))
  (bind _asfort_player_pawn (Utilities|Casting|CastToFortPlayerPawn Target)
    (:then
      (return (not (GameplayTags|HasAnyMatchingGameplayTags _asfort_player_pawn _avoid_detection_tags_container))))
    (:CastFailed
      (bind _asnpc_pawn_parent (Utilities|Casting|CastToNPC_Pawn_Parent Target)
        (:then
          (return (not (GameplayTags|HasAnyMatchingGameplayTags _asnpc_pawn_parent _avoid_detection_tags_container))))
        (:CastFailed
          (return false))))))

(fn OnRep_bHitIsValid
  (Networking|SwitchHasAuthority
    (:Remote
      (CallFunction|OnImpactControlBusMix))))

(fn OnRep_DemonMaskFireOwnerWeapon
  (Networking|SwitchHasAuthority
    (:Remote
      (CallFunction|PlayFireballShootSound))))

(fn OnRep_RepLaserGuidedSpeedValues
  (if (Variables|Default|GetRepLaserGuidedSpeedValues)
    (Projectile|SetHomingStyle 0 0 "LaserTargeted" false :!class /Script/FortniteGame.FortProjectileMovementInterface)
    (bind _output_get (Variables|Homing|SetPrjMaxSpeed_LaserGuided (ScalableFloat|GetValueatLevel (Variables|Homing|GetRow_MaxSpeed_LaserGuided))))
    (Variables|/Script/Engine.ProjectileMovementComponent|SetMaxSpeed _output_get (Variables|Movement|GetProjectileMovementComponent))
    (Utilities|Time|ClearandInvalidateTimerbyHandle (Variables|Default|GetDelaySpeedIncreaseTimer))
    (Variables|Default|SetRepLaserGuidedSpeedValues false)))

(fn OnRep_RepSpeedValues
  (if (Variables|Default|GetRepSpeedValues)
    (Projectile|SetHomingStyle 0 0 "LockOn" false :!class /Script/FortniteGame.FortProjectileMovementInterface)
    (bind _output_get (Variables|Homing|SetPrjMaxSpeed (ScalableFloat|GetValueatLevel (Variables|Homing|GetRow_MaxSpeed))))
    (Variables|/Script/Engine.ProjectileMovementComponent|SetMaxSpeed _output_get (Variables|Movement|GetProjectileMovementComponent))
    (Utilities|Time|ClearandInvalidateTimerbyHandle (Variables|Default|GetDelaySpeedIncreaseTimer))
    (Variables|Default|SetRepSpeedValues false)))

(fn PeriodicCheckForHomingTarget
  (bind _asbpi_demon_mask_fire_weapon (Utilities|Casting|CastToBPI_DemonMaskFireWeapon (Variables|Default|GetDemonMaskFireOwnerWeapon)))
  (bind _homingtarget (BPI_DemonMaskFireWeapon|GetWeaponHomingTarget _asbpi_demon_mask_fire_weapon))
  (Utilities|IsValid _homingtarget
    (:Is%20Not%20Valid
      (Default|FindAndSetHomingTarget))))

(fn SetHomingVars
  (bind _homingtarget (CallFunction|GetHomingTarget))
  (bind _homingtarget_2 (CallFunction|GetHomingTarget))
  (bind _self self)
  (Networking|SwitchHasAuthority
    (:Authority
      (if (not (Delivery|IsBeingKilled))
        (bind _output_get (Variables|Default|SetNew%20Homing%20Style "LockOn"))
        (Projectile|SetHomingStyle 0 0 _output_get false :!class /Script/FortniteGame.FortProjectileMovementInterface)
        (Utilities|IsValid _homingtarget_2
          (:Is%20Valid
            (LocalVariables|SetLocal_SetFastTurnForCloseTarget (> (Transformation|GetSquaredDistanceTo self _homingtarget_2) 6.4e+05)))
          (:Is%20Not%20Valid
            _))
        (Projectile|SetHomingTurnSpeedMin 0 _self (Math|Float|SelectFloat 2e+02 (ScalableFloat|GetValueatLevel (Variables|Homing|GetRow_HomingMinTurnSpeed)) (LocalVariables|GetLocal_SetFastTurnForCloseTarget)) :!class /Script/FortniteGame.FortProjectileMovementInterface)
        (Projectile|SetHomingTurnSpeedMax 0 _self (Math|Float|SelectFloat 2e+02 (ScalableFloat|GetValueatLevel (Variables|Homing|GetRow_HomingMaxTurnSpeed)) (LocalVariables|GetLocal_SetFastTurnForCloseTarget)) :!class /Script/FortniteGame.FortProjectileMovementInterface)
        (Projectile|SetHomingTimeUntilMaxTurnSpeed 0 _self (Math|Float|SelectFloat 0.0 (ScalableFloat|GetValueatLevel (Variables|Homing|GetRow_HomingSpeedRampTime)) (LocalVariables|GetLocal_SetFastTurnForCloseTarget)) :!class /Script/FortniteGame.FortProjectileMovementInterface)
        (bind _asfort_player_pawn (Utilities|Casting|CastToFortPlayerPawn (CallFunction|GetHomingTarget))
          (:then
            (Projectile|SetHomingTargetOffset _self (select (Health|IsDBNO _asfort_player_pawn) (Variables|Homing|GetDBNO%20Target%20Homing%20Offset) (Variables|Homing|GetTargetHomingOffset)) :!class /Script/FortniteGame.FortProjectileMovementInterface))
          (:CastFailed
            (Projectile|SetHomingTargetOffset _self (Variables|Homing|GetTargetHomingOffset) :!class /Script/FortniteGame.FortProjectileMovementInterface)))
        (Utilities|IsValid _homingtarget
          (:Is%20Valid
            (Projectile|SetHomingTarget 0 self _homingtarget 0 :!class /Script/FortniteGame.FortProjectileMovementInterface)))))))

(fn SetProjectileVaraiblesFromHotfixRows
  (bind _row_basedmovinginitialspeedmulti (Variables|Speed|GetRow_BasedMovingInitialSpeedMulti))
  (Variables|Homing|SetPrjMaxSpeed (ScalableFloat|GetValueatLevel (Variables|Homing|GetRow_MaxSpeed)))
  (Variables|Homing|SetHomingWidthAndHeight (ScalableFloat|GetValueatLevel (Variables|Homing|GetDataHomingRectHeightandWidth)))
  (Variables|Homing|SetHomingRange (ScalableFloat|GetValueatLevel (Variables|Homing|GetDataHomingRange)))
  (Variables|Homing|SetSeekingHomingWidthAndHeight (ScalableFloat|GetValueatLevel (Variables|Homing|GetRow_SeekingRectHeightAndWidth)))
  (bind _returnvalue (GameplayTags|HasAnyMatchingGameplayTags(Message) (Variables|Default|GetFort%20Player%20Pawn) (Variables|Default|GetOn%20Train%20Tag%20Container)))
  (Variables|Speed|SetInitialAccelSpeedMulti (ScalableFloat|GetValueatLevel (select _returnvalue _row_basedmovinginitialspeedmulti (Variables|Speed|GetRow_InitialAccelSpeedMulti))))
  (Variables|Speed|SetIncreasedAccelSpeedMulti (ScalableFloat|GetValueatLevel (select _returnvalue _row_basedmovinginitialspeedmulti (Variables|Speed|GetRow_IncreasedAccelSpeedMulti)))))

(fn TryToApplyFXRadialForce
  (params (SpawnTransform :type Transform))
  (if (Math|Boolean|ANDBoolean (not (Fortnite|IsMobilePlatform)) (not (Class|PlatformHelpersLibrary|IsPcPerfMode)) (> (Rendering|GetRenderingDetailMode) 1))
    _))

(fn UpdateHomingTarget
  (params (HomingTarget :type Actor%20Object%20Reference))
  (bind _asbpi_demon_mask_fire_weapon (Utilities|Casting|CastToBPI_DemonMaskFireWeapon (Variables|Default|GetDemonMaskFireOwnerWeapon)))
  (BPI_DemonMaskFireWeapon|SetWeaponHomingTarget _asbpi_demon_mask_fire_weapon HomingTarget))

(fn ConstructionScript)
